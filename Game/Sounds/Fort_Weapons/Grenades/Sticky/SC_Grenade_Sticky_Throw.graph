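# Sound cue for the sticky grenade throw, reusing three frag grenade throw recordings.
# An equally weighted random node, excluded from branch culling, picks a recording and feeds a default-setting modulator.
# A quality-level node at the root routes that same chain to its first three inputs, leaving the fourth empty.
# It uses explosive throw attenuation and the player foley sound class, with a concurrency override that prevents new instances at the limit.
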SoundCue /Game/Sounds/Fort_Weapons/Grenades/Sticky/SC_Grenade_Sticky_Throw
root=QualityLevel_0
nodes [
 {id=Modulator_0, children [Random_1]}
 {id=QualityLevel_0, children [Modulator_0,Modulator_0,Modulator_0,null]}
 {id=Random_1, children [WavePlayer_4,WavePlayer_5,WavePlayer_6], properties {bShouldExcludeFromBranchCulling=true, weights [1,1,1]}}
 {id=WavePlayer_4, properties {soundWaveAssetPtr=/Game/Sounds/Fort_Weapons/Grenades/Frag/Grenade_Frag_Throw_01}}
 {id=WavePlayer_5, properties {soundWaveAssetPtr=/Game/Sounds/Fort_Weapons/Grenades/Frag/Grenade_Frag_Throw_02}}
 {id=WavePlayer_6, properties {soundWaveAssetPtr=/Game/Sounds/Fort_Weapons/Grenades/Frag/Grenade_Frag_Throw_03}}
]
properties {
 attenuationSettings=/Game/Sounds/Attenuation/Fort_Explosive_Fuse_Prj_Throw_Att
 bOverrideConcurrency=true
 maxDistance=13000
 soundClassObject=&Fort_Player_Foley
 volumeMultiplier=1
 concurrencyOverrides {resolutionRule=PreventNew}
}
